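; BP_CharacterGradientTest is a material gym preview actor for testing character gradient materials on a skeletal mesh.
; Its construction script calls UpdateMesh with the PreviewMesh variable, which assigns that mesh to the SkeletalMesh component.
; UpdateMesh then loops over every material slot and overrides it with the PreviewMaterial, so the preview refreshes at construction.
; The BeginPlay, overlap and Tick events are empty.
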(event EventBeginPlay)

(event Collision|EventActorBeginOverlap
  (params (OtherActor :type Actor%20Object%20Reference)))

(event EventTick
  (params (DeltaSeconds :type Float%20%28single-precision%29)))

(fn UpdateMesh
  (params (New%20SkelMesh :type Skeletal%20Mesh%20Object%20Reference))
  (Components|SkeletalMesh|SetSkeletalMeshAsset (Variables|Default|GetSkeletalMesh) New%20SkelMesh)
  (for _index (range (Rendering|Material|GetNumMaterials (Variables|Default|GetSkeletalMesh)))
    (Rendering|Material|SetMaterial (Variables|Default|GetSkeletalMesh) _index (Variables|Default|GetPreviewMaterial))))

(fn ConstructionScript
  (CallFunction|UpdateMesh (Variables|Default|GetPreviewMesh)))
