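# Unlit MaterialGym test material that visualizes MF_CharacterGradients on skeletal meshes.
# Arm and rump parameters feed the function, with arms masked by a pre-skinned-position cylinder or texture channel.
# Static switches show the gradient raw, as a stepped debug view, or as a looping animated band distorted by cloud noise projected on pre-skinned position.
# The "Radial (true) or Linear (false)" switch receives RadialGradient on both inputs, so it changes nothing.
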Material /Game/Global/Materials/MaterialGym/Materials/CharacterGradients/Materials/M_CharacterGradientTest
graph {
 properties {bUsedWithMeshDeformer=true, bUsedWithSkeletalMesh=true, pixelDepthOffsetMode=PDOM_Legacy, shadingModel=MSM_Unlit, usedShadingModels=Unlit}
 nodes [
  n1 = PreSkinnedPosition()
  n2 = SphereMask(
   A=ComponentMask(None=VertexInterpolator(VS=@n1).PS) {g=true, r=true}
   B=Constant2Vector()
   Radius=ScalarParameter() {defaultValue=30, parameterName=CharacterGradient_RumpRadius, sortPriority=1000}
   Hardness=ScalarParameter() {defaultValue=0.8, parameterName=CharacterGradient_RumpFalloffHardness, sliderMax=1, sortPriority=1000}
  ) {attenuationRadius=30, hardnessPercent=80}
  n3 = TextureSampleParameter2D() {
   parameterName=Arms Mask
   samplerType=SAMPLERTYPE_Masks
   texture=/Game/Characters/Player/Male/Medium/Bodies/M_MED_ZirconSweep/Textures/T_M_MED_ZirconSweep_Body_TechArtFX_Masks
  }
  n4 = MaterialFunctionCall!(
   Local Position
   ArmBalance=ScalarParameter() {defaultValue=0.5, parameterName=Arm Balance, sliderMax=1.5, sortPriority=-10}
   RumpBalance=ScalarParameter() {defaultValue=0.5, parameterName=Rump Balance, sliderMax=1.5, sortPriority=-10}
   Arms Mask=StaticSwitchParameter(True=OneMinus(None=@n2), False=@n3.G) {group=1_Features, parameterName=Use Cylinder Mask, sortPriority=320}
   Arm_GradientEndPosition
   Arm_GradientBeginPosition
   Arm_GradientDirection
   Rump_GradientEndHeight=ScalarParameter() {parameterName=RumpGradient_FeetHeight, sliderMax=180, sortPriority=64}
   Rump_GradientBeginHeight=ScalarParameter() {defaultValue=140, parameterName=RumpGradient_TopHeight, sliderMax=180, sortPriority=64}
   Radial_CenterHeight=ScalarParameter() {defaultValue=0.5, parameterName=Radial_CenterHeightPercent, sliderMax=1, sortPriority=16}
   Radial_HorizontalCompression=ScalarParameter() {defaultValue=1, parameterName=Radial_HorizontalCompression, sliderMax=2, sliderMin=0.01, sortPriority=16}
   Disable Arm Gradients=StaticBoolParameter() {group=1_Features, parameterName=Disable Arm Gradient}
   Disable Rump Gradient=StaticBoolParameter() {group=1_Features, parameterName=Disable Rump Gradient}
  ) {materialFunction=/Game/Global/MaterialFunctions/TechArt/MF_CharacterGradients}
  n5 = StaticSwitchParameter(True=@n4.RadialGradient, False=@n4.RadialGradient) {group=1_Features, parameterName="Radial (true) or Linear (false)"}
  n6 = Frac(None=Multiply(A=Time()) {constB=0.4})
  n7 = TextureObject() {texture=&AtmosphericCloudNoise01}
  n8 = VertexInterpolator(VS=PreSkinnedPosition())
  n9 = Multiply(A=ConstantBiasScale(None=MaterialFunctionCall(TextureObject=@n7, WorldPosition=@n8.PS) {materialFunction=/Game/Weapons/WeaponSkins/Wraps/MaterialFunctions/WorldAlignedTexture}.XY Texture) {bias=-0.5, scale=2}) {constB=0.05}
  n10 = Subtract(A=Subtract(A=Multiply(A=@n6) {constB=1.25}) {constB=0.25}, B=Multiply(A=Add(A=@n5, B=@n9)) {constB=0.8})
  n11 = Multiply(A=@n10) {constB=3}
  n12 = Power(Base=OneMinus(None=Saturate(None=@n11))) {constExponent=12}
  n13 = Multiply(A=@n10) {constB=40}
  n14 = Power(Base=Saturate(None=Add(A=@n13)))
  StaticSwitchParameter(
   True=MaterialFunctionCall!(
    Gradient Input=@n5
    Clamp Input=StaticBoolParameter() {defaultValue=true, group=1_Features, parameterName=Clamp Debug View, sortPriority=34}
    Num Steps=ScalarParameter() {defaultValue=10, parameterName=DebugViewSteps, sortPriority=500}
   ) {materialFunction=/Game/Global/MaterialFunctions/TechArt/MF_GradientSteppedDebugView}.Result
   False=StaticSwitchParameter(True=Multiply(A=@n12, B=@n14), False=@n5) {group=1_Features, parameterName=Animated, sortPriority=40}
  ) {group=1_Features, parameterName=Debug View, sortPriority=33}
  VectorParameter() {parameterName=RumpGradientStart, defaultValue=<rgba 0 8 140 0>}
 ]
}
